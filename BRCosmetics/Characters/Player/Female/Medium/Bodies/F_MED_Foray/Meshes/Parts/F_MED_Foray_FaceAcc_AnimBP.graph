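# Anim Blueprint for the F_MED_Foray face accessory that poses its hair bones on top of the pose copied from the attached parent mesh.
# One branch simulates hair with RigidBody physics; the other uses weighted constraints to head, neck, spine and clavicle bones.
# Head yaw relative to neck_01 drives range-mapped additive bone corrections in both branches.
# A one-time check selects constraints when a backpack is worn or rigid bodies are disabled.
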AnimBlueprint /BRCosmetics/Characters/Player/Female/Medium/Bodies/F_MED_Foray/Meshes/Parts/F_MED_Foray_FaceAcc_AnimBP
target_skeleton=&/F_MED_Foray_FaceAcc_Skeleton
graphs [
 AnimGraph {nodes [
   Output Pose(Result=Blend Poses by bool(
     BlendPose_0=Use cached pose 'NoneSimHairPose'()
     BlendPose_1=Use cached pose 'SimHairPose'()
     bActiveValue=Get hairConstraintON() -> (hairConstraintON:bool)
    ) {blendType=Linear}) {layerGroup=None}
   n1 = "Constraint - dyn_hair_a (PP)"!(ComponentPose=Local To Component(LocalPose=Use cached pose 'BasePose'()), Alpha=1, ConstraintWeights_0=0.9, ConstraintWeights_1:float=0.1) -> (Pose) {
    alphaBoolBlend {blendOption=Linear}
    boneToModify {boneName=dyn_hair_a}
    constraintSetup [{targetBone {boneName=head}}, {targetBone {boneName=spine_05}}]
    constraintWeights [1,1]
   }
   n2 = "Constraint - dyn_hair_b (PP)"!(ComponentPose=@n1, Alpha=1, ConstraintWeights_0=0.3, ConstraintWeights_1:float=0.7) -> (Pose) {
    alphaBoolBlend {blendOption=Linear}
    boneToModify {boneName=dyn_hair_b}
    constraintSetup [{targetBone {boneName=head}}, {targetBone {boneName=spine_05}}]
    constraintWeights [1,1]
   }
   n3 = "Constraint - dyn_hair_c (PPP)"!(ComponentPose=@n2, Alpha=1, ConstraintWeights_0=0.15, ConstraintWeights_1:float=0.7, ConstraintWeights_2:float=0.15) -> (Pose) {
    alphaBoolBlend {blendOption=Linear}
    boneToModify {boneName=dyn_hair_c}
    constraintSetup [{targetBone {boneName=clavicle_r}}, {targetBone {boneName=spine_05}}, {targetBone {boneName=clavicle_l}}]
    constraintWeights [1,1,1]
   }
   n4 = "Constraint - dyn_hair_b_l (PPP)"!(ComponentPose=@n3, Alpha=1, ConstraintWeights_0=0.15, ConstraintWeights_1:float=0.65, ConstraintWeights_2:float=0.2) -> (Pose) {
    alphaBoolBlend {blendOption=Linear}
    boneToModify {boneName=dyn_hair_b_l}
    constraintSetup [{targetBone {boneName=head}}, {targetBone {boneName=spine_05}}, {targetBone {boneName=clavicle_l}}]
    constraintWeights [1,1,1]
   }
   n5 = "Constraint - dyn_hair_c_l (PP)"!(ComponentPose=@n4, Alpha=1, ConstraintWeights_0=0.4, ConstraintWeights_1:float=0.6) -> (Pose) {
    alphaBoolBlend {blendOption=Linear}
    boneToModify {boneName=dyn_hair_c_l}
    constraintSetup [{targetBone {boneName=spine_05}}, {targetBone {boneName=clavicle_l}}]
    constraintWeights [1,1]
   }
   n6 = "Constraint - dyn_hair_b_r (PPPP)"!(
    ComponentPose=@n5
    Alpha=1
    ConstraintWeights_0=0.1
    ConstraintWeights_1:float=0.1
    ConstraintWeights_2:float=0.1
    ConstraintWeights_3:float=0.7
   ) -> (Pose) {
    alphaBoolBlend {blendOption=Linear}
    boneToModify {boneName=dyn_hair_b_r}
    constraintSetup [{targetBone {boneName=clavicle_r}}, {targetBone {boneName=neck_02}}, {targetBone {boneName=neck_01}}, {targetBone {boneName=spine_05}}]
    constraintWeights [1,1,1,1]
   }
   n7 = "Constraint - dyn_hair_c_r (PP)"!(ComponentPose=@n6, Alpha=1, ConstraintWeights_0=0.7, ConstraintWeights_1:float=0.3) -> (Pose) {
    alphaBoolBlend {blendOption=Linear}
    boneToModify {boneName=dyn_hair_c_r}
    constraintSetup [{targetBone {boneName=clavicle_r}}, {targetBone {boneName=spine_05}}]
    constraintWeights [1,1]
   }
   n8 = Get head_fwdBwd_rot() -> (head_fwdBwd_rot:double)
   n9 = "Transform (Modify) Bone - Bone: dyn_hair_b"(ComponentPose=@n7, Alpha=@n8, Translation="0,-5,0", Rotation="0,20,0", Scale="1,1,1") {
    rotationMode=BMM_Additive
    rotationSpace=BCS_BoneSpace
    scaleMode=BMM_Additive
    scaleSpace=BCS_BoneSpace
    translationMode=BMM_Additive
    translationSpace=BCS_BoneSpace
    alphaBoolBlend {blendOption=Linear}
    alphaScaleBiasClamp {bMapRange=true, inRange {max=0, min=-30}, outRange {max=0, min=1}}
    boneToModify {boneName=dyn_hair_b}
    rotation {yaw=20}
    translation {y=-5}
   }
   n10 = "Transform (Modify) Bone - Bone: dyn_hair_b_l"(ComponentPose=@n9, Alpha=@n8, Translation="0,0,-3", Rotation="15,10,0", Scale="1,1,1") {
    rotationMode=BMM_Additive
    rotationSpace=BCS_BoneSpace
    scaleMode=BMM_Additive
    scaleSpace=BCS_BoneSpace
    translationMode=BMM_Additive
    translationSpace=BCS_BoneSpace
    alphaBoolBlend {blendOption=Linear}
    alphaScaleBiasClamp {bMapRange=true, inRange {max=0, min=-30}, outRange {max=0, min=1}}
    boneToModify {boneName=dyn_hair_b_l}
    rotation {pitch=15, yaw=10}
    translation {z=-3}
   }
   Save cached pose 'NoneSimHairPose'(Pose=Component To Local(ComponentPose=@n10)) {cacheName=NoneSimHairPose}
   Save cached pose 'BasePose'(Pose=Copy Pose From Mesh() {bUseAttachedParent=true}) {cacheName=BasePose}
   n11 = Get head_fwdBwd_rot()
   n12 = "Transform (Modify) Bone - Bone: dyn_hair_a_base"(
    ComponentPose=Local To Component(LocalPose=Use cached pose 'BasePose'())
    Alpha=@n11
    Translation="-5,0,0"
    Rotation="0,-40,0"
    Scale="1,1,1"
   ) {
    rotationMode=BMM_Additive
    rotationSpace=BCS_BoneSpace
    scaleMode=BMM_Additive
    scaleSpace=BCS_BoneSpace
    translationMode=BMM_Additive
    translationSpace=BCS_BoneSpace
    alphaBoolBlend {blendOption=Linear}
    alphaScaleBiasClamp {bMapRange=true, inRange {max=0, min=-30}, outRange {max=0, min=1}}
    boneToModify {boneName=dyn_hair_a_base}
    rotation {yaw=-40}
    translation {x=-5}
   }
   n13 = "Transform (Modify) Bone - Bone: dyn_hair_a_base_l"(ComponentPose=@n12, Alpha=@n11, Translation="0,0,-3", Rotation="-20,5,0", Scale="1,1,1") {
    rotationMode=BMM_Additive
    rotationSpace=BCS_BoneSpace
    scaleMode=BMM_Additive
    scaleSpace=BCS_BoneSpace
    translationMode=BMM_Additive
    translationSpace=BCS_BoneSpace
    alphaBoolBlend {blendOption=Linear}
    alphaScaleBiasClamp {bMapRange=true, inRange {max=0, min=-30}, outRange {max=0, min=1}}
    boneToModify {boneName=dyn_hair_a_base_l}
    rotation {pitch=-20, yaw=5}
    translation {z=-3}
   }
   n14 = "Transform (Modify) Bone - Bone: dyn_hair_a_base_r"(ComponentPose=@n13, Alpha=@n11, Translation="0,0,3", Rotation="-40,0,0", Scale="1,1,1") {
    rotationMode=BMM_Additive
    rotationSpace=BCS_BoneSpace
    scaleMode=BMM_Additive
    scaleSpace=BCS_BoneSpace
    translationMode=BMM_Additive
    translationSpace=BCS_BoneSpace
    alphaBoolBlend {blendOption=Linear}
    alphaScaleBiasClamp {bMapRange=true, inRange {max=0, min=-30}, outRange {max=0, min=1}}
    boneToModify {boneName=dyn_hair_a_base_r}
    rotation {pitch=-40}
    translation {z=3}
   }
   n15 = Get head_fwdBwd_rot()
   n16 = "Transform (Modify) Bone - Bone: dyn_hair_a_base"(ComponentPose=@n14, Alpha=@n15, Translation="2,0,0", Rotation="0,10,0", Scale="1,1,1") {
    rotationMode=BMM_Additive
    rotationSpace=BCS_BoneSpace
    scaleMode=BMM_Additive
    scaleSpace=BCS_BoneSpace
    translationMode=BMM_Additive
    translationSpace=BCS_BoneSpace
    alphaBoolBlend {blendOption=Linear}
    alphaScaleBiasClamp {bMapRange=true, inRange {max=20}}
    boneToModify {boneName=dyn_hair_a_base}
    rotation {yaw=10}
    translation {x=2}
   }
   n17 = "Transform (Modify) Bone - Bone: dyn_hair_a_base_l"(ComponentPose=@n16, Alpha=@n15, Translation="0,0,0", Rotation="15,0,0", Scale="1,1,1") {
    rotationMode=BMM_Additive
    rotationSpace=BCS_BoneSpace
    scaleMode=BMM_Additive
    scaleSpace=BCS_BoneSpace
    translationMode=BMM_Additive
    translationSpace=BCS_BoneSpace
    alphaBoolBlend {blendOption=Linear}
    alphaScaleBiasClamp {bMapRange=true, inRange {max=20}}
    boneToModify {boneName=dyn_hair_a_base_l}
    rotation {pitch=15}
   }
   n18 = "Transform (Modify) Bone - Bone: dyn_hair_a_base_r"(ComponentPose=@n17, Alpha=@n15, Translation="0,0,0", Rotation="15,0,0", Scale="1,1,1") {
    rotationMode=BMM_Additive
    rotationSpace=BCS_BoneSpace
    scaleMode=BMM_Additive
    scaleSpace=BCS_BoneSpace
    translationMode=BMM_Additive
    translationSpace=BCS_BoneSpace
    alphaBoolBlend {blendOption=Linear}
    alphaScaleBiasClamp {bMapRange=true, inRange {max=20}}
    boneToModify {boneName=dyn_hair_a_base_r}
    rotation {pitch=15}
   }
   n19 = RigidBody(ComponentPose=@n18) {
    bClampLinearTranslationLimitToRefPose=true
    bOverrideWorldGravity=true
    overridePhysicsAsset=&/F_MED_Foray_FaceAcc_Physics
    simulationSpace=BaseBoneSpace
    alphaBoolBlend {blendOption=Linear}
    baseBoneRef {boneName=spine_05}
    componentLinearAccScale=<0.35 0.35 0.35>
    componentLinearVelScale=<0.35 0.35 0.35>
    overrideWorldGravity {z=-980}
   }
   Save cached pose 'SimHairPose'(Pose=Component To Local(ComponentPose=@n19)) {cacheName=SimHairPose}
  ]}
 EventGraph {nodes [
   Event BlueprintUpdateAnimation()
   > n1 = Sequence!(execute) -> (then_0, then_1, then_2:Exec, then_3:Exec, then_4:Exec, then_5:Exec, then_6:Exec)
   n2 = Get playerPawn() -> (playerPawn:Fort Player Pawn Athena Object Reference)
   Is Valid(exec=@n1.then_1, InputObject=@n2)
   > Do Once()
   > n3 = Set IsBackpackOn(IsBackpackOn:bool="Not Equal (Object)"!(A:Object Reference=GetSkeletalMeshForPartType(self=@n2, PartType=Backpack), B:Object Reference) -> (ReturnValue:bool)) -> (Output_Get:bool)
   n4 = Get bRigidBodyEnabled()
   > n5 = Branch(Condition=OR Boolean(A=@n3, B=NOT Boolean(A=@n4)))
   > Set hairConstraintON(hairConstraintON:bool=true) -> (Output_Get:bool)
   Set hairConstraintON(execute=@n5.else, hairConstraintON=false)
   n6 = GetDeltaTransformFromRefPose(self=GetOwnerSkeletalMeshComponent(), BoneName=head, BaseName=neck_01)
   Set head_fwdBwd_rot(execute=@n1.then_0, head_fwdBwd_rot:double=BreakRotator(InRot=BreakTransform(InTransform=@n6).Rotation).Yaw) -> (Output_Get:double)
   Event BlueprintInitializeAnimation()
   > Set MeshToCopy(MeshToCopy:Skeletal Mesh Component Object Reference=GetOwnerSkeletalMeshComponent()) -> (Output_Get:Skeletal Mesh Component Object Reference)
   n7 = TryGetPawnOwner()
   > Is Valid(InputObject=@n7)
   > n8 = Cast To FortPlayerPawnAthena(Object=@n7)
   > Set playerPawn(playerPawn:Fort Player Pawn Athena Object Reference=@n8) -> (Output_Get:Fort Player Pawn Athena Object Reference)
   TryGetPawnOwner()
  ]}
]
variables [
 hairConstraintON {type=boolean}
 head_fwdBwd_rot {type=number}
 IsBackpackOn {type=boolean}
 MeshToCopy {type=SkeletalMeshComponent}
 playerPawn {type=FortPlayerPawnAthena}
]
